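# Material function for the ShinToll Keytar that adds animated, masked red and cyan pulse-line graphics to the incoming emissive.
# Two parallax layers use BumpOffset, rotated UVs and a curve-atlas row to shape sine-animated lines, clipped by mask texture channels.
# Moving highlight bands brighten both colors over the base color.
# Animation blends Time with named reroutes, which include music-clock measure progress and a timestamp-valid flag from a parameter collection.
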MaterialFunction /SparksCosmetics/Instrument/Keytar/ShinToll/Materials/MF_ShinToll_Keytar
graph {nodes [
  n1 = GetMaterialAttributes(None=FunctionInput() {inputName=MA, inputType=FunctionInput_MaterialAttributes}) {attributeGetTypes [#1, #2]}
  n2 = ComponentMask(None=TextureCoordinate() {vTiling=2}) {g=true}
  n3 = Time()
  n4 = NamedRerouteUsage()
  n5 = NamedRerouteUsage()
  n6 = Fmod(A=Add(A=LinearInterpolate(A=@n3, B=@n4, Alpha=@n5)) {constB=0.84}, B=Constant() {r=1})
  n7 = Constant() {r=-0.73}
  n8 = Add(A=Add(A=@n2, B=@n6) {constB=-0.5}, B=@n7) {constB=-0.5}
  n9 = Multiply(A=@n8) {constB=10}
  n10 = Add(A=OneMinus(None=@n9)) {constB=-0.3}
  n11 = Add(A=Add(A=@n2, B=Multiply(A=@n6) {constB=-1}) {constB=-0.5}, B=@n7) {constB=-1}
  n12 = Multiply(A=@n11) {constB=10}
  n13 = Add(A=OneMinus(None=@n12)) {constB=-0.3}
  n14 = Add(A=Saturate(None=Multiply(A=@n9, B=@n10)), B=Saturate(None=Multiply(A=@n12, B=@n13)))
  n15 = Multiply(A=Subtract(A=@n8) {constB=-0.05}) {constB=2}
  n16 = Add(A=OneMinus(None=@n15)) {constB=-0.3}
  n17 = Multiply(A=Add(A=@n11) {constB=0.2}) {constB=2}
  n18 = Add(A=OneMinus(None=@n17)) {constB=-0.3}
  n19 = Add(A=Saturate(None=Multiply(A=@n15, B=@n16)), B=Saturate(None=Multiply(A=@n17, B=@n18)))
  n20 = Add(A=Saturate(None=Multiply(A=@n14) {constB=10}), B=Saturate(None=Multiply(A=@n19) {constB=3}))
  n21 = LinearInterpolate(A=Constant3Vector() {constant=<rgba 0.9449 0.1726 0.1726 1>}, B=Constant3Vector() {constant=<rgba 0.9387 0.4801 0.4653 1>}, Alpha=@n20)
  n22 = LinearInterpolate(A=Constant3Vector() {constant=<rgba 0 0.65 1 1>}, B=Constant3Vector() {constant=<rgba 0.57 0.97 1 1>}, Alpha=@n20)
  n23 = Constant()
  n24 = Constant() {r=1}
  n25 = TextureCoordinate() {coordinateIndex=1}
  n26 = ScalarParameter() {defaultValue=-0.2, group=DryErase, parameterName=A_BumpOffsetHeight, sortPriority=-2}
  n27 = Subtract(A=Subtract(A=BumpOffset(Coordinate=@n25, Height=@n26), B=Constant2Vector() {g=-0.03, r=0.03}) {constB=0.5}) {constB=0.5}
  n28 = MaterialFunctionCall(UVs=Add(A=Divide(A=@n27) {constB=0.85}) {constB=0.5}, "Rotation Angle (0-1)"=Constant() {r=0.25}) {materialFunction=&CustomRotator}
  n29 = ComponentMask(None=@n28.Rotated Values) {g=true}
  n30 = CurveAtlasRowParameter(CurveTime=ComponentMask(None=@n28.Rotated Values) {r=true}) {
   atlas=&/CA_ShinTollPulse
   curve=&/CRV_ShinToll_Keytar_Anim
   parameterName=Pulse
  }
  n31 = CurveAtlasRowParameter() {
   atlas=&CA_ShinTollPulse
   curve=&CRV_ShinToll_Keytar_Anim
   parameterName=Pulse
  }
  n32 = Multiply(A=Time() {period=1}) {constB=2}
  n33 = Constant() {r=2}
  n34 = LinearInterpolate(A=Saturate(None=Fmod(A=@n32, B=@n33)), B=NamedRerouteUsage(), Alpha=NamedRerouteUsage())
  n35 = Multiply(A=Saturate(None=Sine(None=@n34))) {constB=-0.55}
  n36 = Multiply(A=Subtract(A=@n30.R, B=@n31.R) {constB=0.5}, B=@n35) {constB=0.3}
  n37 = If(A=@n35, "A > B"=Constant(), "A < B"=Subtract(A=@n31.R, B=@n31.G))
  n38 = If(A=@n29, B=Subtract(A=Add(A=@n36, B=@n31.R) {constB=0.49}, B=@n37) {constB=0.1}, "A > B"=@n23, "A < B"=@n24)
  n39 = Multiply(A=Subtract(A=@n30.G, B=@n31.G) {constB=0.487}, B=@n35) {constB=0.3}
  n40 = If(A=@n29, B=Add(A=Add(A=@n39, B=@n31.G) {constB=0.497}, B=@n37), "A > B"=@n24, "A < B"=@n23)
  n41 = Multiply(A=@n38, B=@n40)
  n42 = Multiply(A=OneMinus(None=@n38), B=OneMinus(None=@n40))
  n43 = TextureSample() {texture=/SparksCosmetics/Instrument/Keytar/ShinToll/Textures/T_ShinToll_Keytar_Mask}
  n44 = Saturate(None=Subtract(A=If(A=@n35, "A > B"=@n41, "A < B"=@n42), B=OneMinus(None=@n43.G)))
  n45 = Constant()
  n46 = Constant() {r=1}
  n47 = TextureCoordinate() {coordinateIndex=1}
  n48 = ScalarParameter() {defaultValue=-0.5, group=DryErase, parameterName=B_BumpOffsetHeight, sortPriority=-2}
  n49 = Subtract(A=Subtract(A=BumpOffset(Coordinate=@n47, Height=@n48) {heightRatio=0.005}, B=Constant2Vector() {g=-0.015, r=0.005}) {constB=0.5}) {constB=0.5}
  n50 = MaterialFunctionCall(UVs=Add(A=Divide(A=@n49) {constB=0.3}) {constB=0.5}, "Rotation Angle (0-1)"=Constant() {r=0.25}) {materialFunction=&CustomRotator}
  n51 = ComponentMask(None=@n50.Rotated Values) {g=true}
  n52 = CurveAtlasRowParameter(CurveTime=ComponentMask(None=@n50.Rotated Values) {r=true}) {
   atlas=&CA_ShinTollPulse
   curve=&CRV_ShinToll_Keytar_Anim
   parameterName=Pulse
  }
  n53 = Constant() {r=0.5}
  n54 = Time() {period=1}
  n55 = Fmod(A=Multiply(A=Subtract(A=@n54) {constB=0.1}) {constB=2}, B=Constant() {r=2})
  n56 = Multiply(A=LinearInterpolate(A=Saturate(None=@n55), B=NamedRerouteUsage(), Alpha=NamedRerouteUsage())) {constB=-1}
  n57 = Multiply(A=Saturate(None=Sine(None=@n56))) {constB=0.27}
  n58 = Multiply(A=Subtract(A=@n52.R, B=@n53) {constB=0.5}, B=@n57) {constB=0.3}
  n59 = Constant() {r=0.472}
  n60 = If(A=@n57, "A > B"=Constant(), "A < B"=Subtract(A=@n53, B=@n59))
  n61 = If(A=@n51, B=Subtract(A=Add(A=@n58, B=@n53) {constB=0.49}, B=@n60) {constB=0.1}, "A > B"=@n45, "A < B"=@n46)
  n62 = Multiply(A=Subtract(A=@n52.G, B=@n59) {constB=0.487}, B=@n57) {constB=0.3}
  n63 = If(A=@n51, B=Add(A=Add(A=@n62, B=@n59) {constB=0.497}, B=@n60), "A > B"=@n46, "A < B"=@n45)
  n64 = Multiply(A=@n61, B=@n63)
  n65 = Multiply(A=OneMinus(None=@n61), B=OneMinus(None=@n63))
  n66 = Saturate(None=Subtract(A=If(A=@n57, "A > B"=@n64, "A < B"=@n65), B=OneMinus(None=@n43.R)))
  n67 = LinearInterpolate(
   B=LinearInterpolate(A=@n1.BaseColor, B=LinearInterpolate(A=@n21, B=@n22, Alpha=@n44) {constB=0}, Alpha=Add(A=@n44, B=@n66)) {constA=1}
   Alpha=Saturate(None=Add(A=@n43.R, B=@n43.R))
  )
  FunctionOutput(None=SetMaterialAttributes!(MaterialAttributes=@n1.MaterialAttributes, EmissiveColor=Add(A=@n1.EmissiveColor, B=@n67)) {attributeSetTypes [#1]})
  NamedRerouteDeclaration(None=Multiply(A=CollectionParameter() {collection=/Game/Sounds/MusicEventSubsystem/MPC_MusicEventSubsystem_PriorityClock, parameterName=PreviousFrameMusicMeasureProgress}) {constB=4}) {name=Pulse, nodeColor=<rgba 0.0078 0 1 1>}
  NamedRerouteDeclaration(None=CollectionParameter() {collection=&MPC_MusicEventSubsystem_PriorityClock, parameterName=MusicTimestampValid}) {name=IsMusicOn, nodeColor=<rgba 1 0.0408 0 1>}
  NamedRerouteDeclaration(None=CollectionParameter() {collection=&MPC_MusicEventSubsystem_PriorityClock, parameterName=Tempo}) {name=BPM, nodeColor=<rgba 1 0.7734 0 1>}
 ]}
