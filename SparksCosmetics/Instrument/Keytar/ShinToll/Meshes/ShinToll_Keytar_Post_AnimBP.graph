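# Post-process AnimBP for the ShinToll keytar that adds secondary physics motion to its dyn_wire bone.
# AnimDynamics with constrained springs and gravity feeds a lightly weighted SpringBone controller on the same bone.
# Each update zeroes spring alpha above a pawn speed threshold or during melee animations, with half weight in the front end.
# The update also sets a simulation-space variable, using World in the front end.
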AnimBlueprint /SparksCosmetics/Instrument/Keytar/ShinToll/Meshes/ShinToll_Keytar_Post_AnimBP
target_skeleton=&/ShinToll_Keytar_Skeleton
graphs [
 AnimGraph {nodes [
   n1 = Anim Dynamics - Bone: dyn_wire!(SimulationSpace:AnimPhysSimSpaceType Enum, ComponentPose=Local To Component(LocalPose=Input Pose()), Alpha=1, ExternalForce="0,0,0") -> (Pose) {
    angularBiasOverride=1
    angularDampingOverride=1
    angularSpringConstant=50
    bAngularSpring=true
    bLinearSpring=true
    bOverrideAngularBias=true
    bOverrideAngularDamping=true
    bOverrideLinearDamping=true
    bUseGravityOverride=true
    linearDampingOverride=1
    linearSpringConstant=500
    simulationSpace=RootRelative
    alphaScaleBiasClamp {bInterpResult=true, interpSpeedDecreasing=4, interpSpeedIncreasing=4}
    boundBone {boneName=dyn_wire}
    gravityOverride {z=-980}
    physicsBodyDefinitions [{
      boundBone {boneName=dyn_wire}
      boxExtents=<10 10 10>
      constraintSetup {angularLimitsMax=<0 5 2>, angularLimitsMin=<0 -5 -2>, angularTarget=<1 0 0>, linearAxesMax=<0 0.5 4>, linearAxesMin=<-2 -1 -0.5>}
      localJointOffset=<1 1 0>
     }]
    simSpaceSettings {maxAngularAcceleration=45, maxAngularVelocity=8, simSpaceAngularAlpha=0.25}
   }
   n2 = Spring controller - Bone: dyn_wire!(ComponentPose:Component Space Pose Link Structure=@n1, Alpha:float=Get SpringAlpha() -> (SpringAlpha:double)) -> (Pose:Component Space Pose Link Structure) {
    is=title AnimGraphNode_SpringBone
    bRotateY=true
    bTranslateY=false
    springDamping=2.5
    springStiffness=200
    alphaBoolBlend {blendOption=Linear}
    springBone {boneName=dyn_wire}
   }
   Output Pose(Result=Component To Local(ComponentPose=@n2))
  ]}
 EventGraph {nodes [
   Event BlueprintUpdateAnimation()
   > n1 = Sequence()
   > n2 = Branch(Condition=Get bIsInFrontEnd())
   > Set Is Not Melee Anim Float(Is Not Melee Anim Float:float=0.5) -> (Output_Get:float)
   n3 = Get bIsPlayingMeleeAnim()
   Set Is Not Melee Anim Float(execute=@n2.else, Is Not Melee Anim Float="To Float (Boolean)"(InBool=NOT Boolean(A=@n3)))
   n4 = "float > float"(A=Vector Length(A=Get PawnVelocity()), B=1200) {is=title K2Node_CallFunction}
   Set SpringAlpha(
    execute=@n1.then_1
    SpringAlpha:double=float * float!(A=0.015, B=SelectFloat(A=0, B=1, bPickA=@n4), C:float=Get Is Not Melee Anim Float() -> (Is Not Melee Anim Float:float)) -> (ReturnValue) {is=title K2Node_PromotableOperator}
   ) -> (Output_Get:double)
   > n5 = Branch(Condition=Get bIsInFrontEnd())
   > Set simSpace(simSpace:AnimPhysSimSpaceType Enum=World) -> (Output_Get:AnimPhysSimSpaceType Enum)
   Set simSpace(execute=@n5.else)
   TryGetPawnOwner()
  ]}
]
variables [
 Is Not Melee Anim Float {type=number}
 simSpace {type=AnimPhysSimSpaceType}
 SpringAlpha {type=number}
]
